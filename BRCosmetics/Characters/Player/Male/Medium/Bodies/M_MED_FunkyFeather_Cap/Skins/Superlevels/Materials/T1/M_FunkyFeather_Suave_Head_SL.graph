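# Head material for the FunkyFeather Suave Superlevels character skin, wrapping flipbook facial animation in the shared 2023 character parent functions.
# Between PreFX and PostFX, it overrides Base Color and Emissive before MF_FlipbookFace.
# SL parameters supply the base tint, masked emissive detail and fresnel colors through named reroutes.
# Temporal responsiveness follows the face alpha times HasPoseChanged, which the head AnimBP sets for flipbook faces.
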Material /BRCosmetics/Characters/Player/Male/Medium/Bodies/M_MED_FunkyFeather_Cap/Skins/Superlevels/Materials/T1/M_FunkyFeather_Suave_Head_SL
graph {
 properties {
  bUseMaterialAttributes=true
  bUsedWithMeshDeformer=true
  bUsedWithMorphTargets=true
  bUsedWithSkeletalMesh=true
  pixelDepthOffsetMode=PDOM_Legacy
  shadingModel=MSM_FromMaterialExpression
  usedShadingModels=Default Lit | Subsurface | Clear Coat | Subsurface Profile | Cloth
 }
 nodes [
  n1 = MaterialFunctionCall() {materialFunction=&MF_CharacterParent_2023_PreFX}
  n2 = GetMaterialAttributes(None=@n1.Result) {attributeGetTypes [#1, #2]}
  n3 = MaterialFunctionCall!(
   Mouth UV
   Eye UV
   Brow UV
   Material=SetMaterialAttributes!(MaterialAttributes=@n2.MaterialAttributes, Emissive Color=NamedRerouteUsage(), Base Color=NamedRerouteUsage()) {attributeSetTypes [#1, #2]}
  ) {materialFunction=/Game/Global/Materials/Character_Parent_2023/FlipbookFace_Parent_2023/MaterialFunctions/MF_FlipbookFace}
  MaterialFunctionCall(Material=@n3.Material, M=@n1.M, Pre Skinned Local Position=@n1.M, Distance Blend=@n1.M, Fresnel=@n1.M) {materialFunction=&MF_CharacterParent_2023_PostFX}
  NamedRerouteDeclaration(None=@n2.Emissive Color) {name=Emissive, nodeColor=<rgba 0 0.4141 1 1>}
  n4 = LinearInterpolate(A=NamedRerouteUsage(), B=NamedRerouteUsage(), Alpha=NamedRerouteUsage())
  n5 = NamedRerouteUsage()
  NamedRerouteDeclaration(None=LinearInterpolate(A=NamedRerouteUsage(), B=Multiply(A=@n4, B=@n5), Alpha=NamedRerouteUsage())) {name=SL_Emissive, nodeColor=<rgba 0.7344 1 0 1>}
  NamedRerouteDeclaration(None=@n2.Base Color) {name=BaseColor, nodeColor=<rgba 1 0 0.8203 1>}
  n6 = TextureSampleParameter2D() {
   group=SL 01 Global
   parameterName=SL.FX2Mask
   samplerSource=SSM_Wrap_WorldGroupSettings
   samplerType=SAMPLERTYPE_Masks
   sortPriority=-30
   texture=&T_BlackMask
  }
  NamedRerouteDeclaration(None=ChannelMaskParameter(None=NamedRerouteDeclaration(None=@n6.RGB) {name=FX2Mask, nodeColor=<rgba 1 0.4922 0 1>}) {group=SL 01 Global, parameterName=SL.EmissiveDetailMaskChannel, sortPriority=-27}.R) {name=EmissiveDetailMask, nodeColor=<rgba 1 0.6843 0.0075 1>}
  n7 = NamedRerouteUsage()
  n8 = VectorParameter() {group=SL 01 Global, parameterName=SL.BaseColorTint, sortPriority=2, defaultValue=<rgba 1 1 1 1>}
  NamedRerouteDeclaration(None=LinearInterpolate(A=@n7, B=Multiply(A=@n7, B=@n8.RGB), Alpha=NamedRerouteUsage())) {name=SL_Basecolor, nodeColor=<rgba 0 1 0.8594 1>}
  NamedRerouteDeclaration(None=VectorParameter() {group=SL 01 Global, parameterName=SL.EmissiveDetailColor, sortPriority=3}.RGB) {name=EmissiveDetailColor, nodeColor=<rgba 0.1854 0.7843 1 1>}
  NamedRerouteDeclaration(None=ScalarParameter() {group=SL 01 Global, parameterName=SL.EmissiveDetailStrength, sortPriority=3}) {name=EmissiveDetailStrength, nodeColor=<rgba 0.1854 0.7843 1 1>}
  NamedRerouteDeclaration(None=MaterialFunctionCall(
    Brightness=ScalarParameter() {defaultValue=1, group=SL 01 Global, parameterName=SL.EmissiveFresnel_Brightness, sliderMin=4, sortPriority=6}
    Exponent=ScalarParameter() {defaultValue=1.5, group=SL 01 Global, parameterName=SL.EmissiveFresnel_Exponent, sortPriority=7}
    Vertex / Pixel Normal=PixelNormalWS()
   ) {materialFunction=&MF_Fresnel_FX}.Fresnel Inverted) {name=EmissiveFresnel, nodeColor=<rgba 1 0 0.2812 1>}
  NamedRerouteDeclaration(None=VectorParameter() {group=SL 01 Global, parameterName=SL.EmissiveFresnelColor, sortPriority=3}.RGB) {name=EmissiveFresnelColor, nodeColor=<rgba 0.1854 0.7843 1 1>}
  NamedRerouteDeclaration(None=VectorParameter() {group=SL 01 Global, parameterName=SL.EmissiveDetailBaseColor, sortPriority=3}.RGB) {name=EmissiveDetaiBaseColor, nodeColor=<rgba 0.1854 0.7843 1 1>}
  n9 = ScalarParameter() {desc="Set by base head AnimBP if \"IsFlipbookFace\" is true", parameterName=HasPoseChanged}
  TemporalResponsivenessOutput!(None=LinearInterpolate(Alpha=Multiply(A=@n3.FaceAlpha, B=@n9)) {constA=0.4})
 ]
}
